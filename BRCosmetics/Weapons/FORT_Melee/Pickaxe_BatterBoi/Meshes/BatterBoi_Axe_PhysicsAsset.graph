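# Physics asset for the BatterBoi two-handed pickaxe mesh, giving its dangling rope chain collision bodies and joints for physical motion.
# One long capsule covers the handle, and small capsules cover the rope base and three dynamic rope bones.
# Constraints link the bodies in sequence from handle to tip.
# The base-to-handle joint rotates freely, while the rope joints limit swing and allow very little twist, so the rope bends without spinning.
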PhysicsAsset /BRCosmetics/Weapons/FORT_Melee/Pickaxe_BatterBoi/Meshes/BatterBoi_Axe_PhysicsAsset
bodies {
 Melee_TwoHanded_Handle [{shapeType=Capsule, length=51.1023, radius=5.168, center=<-0.0275 2.0229 18.2225>, rotation=<rot 2.0347 -90 -0.0454>}]
 base_rope [{shapeType=Capsule, length=0.505, radius=0.505, shapeName=base_rope_capsule, rotation=<rot 0 -90 0>}]
 dyn_rope_01 [{shapeType=Capsule, length=1.4325, radius=1.4888, shapeName=dyn_rope_01_capsule, center=<1.5518 0 0>, rotation=<rot 90 0 0>}]
 dyn_rope_02 [{shapeType=Capsule, length=1.4638, radius=1.4149, shapeName=dyn_rope_02_capsule, center=<1.3979 0 0>, rotation=<rot 90 0 0>}]
 dyn_rope_03 [{shapeType=Capsule, length=5.4915, radius=2.3413, shapeName=dyn_rope_03_capsule, center=<2.3934 -0.1505 -0.0503>, rotation=<rot 0 0 90>}]
}
constraints [
 {
  bone1Name=base_rope
  bone2Name=Melee_TwoHanded_Handle
  swing1LimitDegrees=45
  swing1Motion=Free
  swing2LimitDegrees=45
  swing2Motion=Free
  twistLimitDegrees=45
  twistMotion=Free
 }
 {bone1Name=dyn_rope_01, bone2Name=base_rope, swing1LimitDegrees=70, swing2LimitDegrees=10, twistLimitDegrees=5}
 {bone1Name=dyn_rope_02, bone2Name=dyn_rope_01, swing1LimitDegrees=50, swing2LimitDegrees=50, twistLimitDegrees=5}
 {bone1Name=dyn_rope_03, bone2Name=dyn_rope_02, swing1LimitDegrees=50, swing2LimitDegrees=50, twistLimitDegrees=5}
]
